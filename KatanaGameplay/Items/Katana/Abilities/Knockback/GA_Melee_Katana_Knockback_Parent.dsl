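; Katana melee knockback parent ability, extending a base knockback ability that launches a target.
; Activation enables high-acceleration net tolerance on the avatar's character movement, then runs parent logic unless the avatar has a Vehicle tag; ending clears the tolerance.
; Launching also grants the owner fall damage immunity until landing unless the target has an ignore tag.
; The building-actor launch event just ends the ability.
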(event Ability|EventActivateAbility
  (bind _outfortplayerpawn (CallFunction|GetAvatarasValidFortPlayerPawn)
    (:ValidFortPlayerPawn
      (bind _asfort_movement_comp_character (Utilities|Casting|CastToFortMovementComp_Character (Pawn|GetMovementComponent _outfortplayerpawn))
        (:then
          (NetTolerance|SetUseHighAccelerationTolerance _asfort_movement_comp_character true "Katana_Knockback"))
        (:CastFailed
          _))
      (bind _returnvalue (GameplayTags|HasAnyMatchingGameplayTags(Message) (Ability|GetAvatarActorfromActorInfo) (GameplayTags|MakeLiteralGameplayTagContainer "(GameplayTags=((TagName=\"Vehicle\")))")))
      (if (not _returnvalue)
        (|CallParentFunction)))
    (:AvatarCastFailed
      _)))

(event Ability|EventOnEndAbility
  (params (bWasCancelled :type Boolean))
  (|CallParentFunction bWasCancelled)
  (bind _outfortplayerpawn (CallFunction|GetAvatarasValidFortPlayerPawn)
    (:ValidFortPlayerPawn
      (bind _asfort_movement_comp_character (Utilities|Casting|CastToFortMovementComp_Character (Pawn|GetMovementComponent _outfortplayerpawn)))
      (NetTolerance|SetUseHighAccelerationTolerance _asfort_movement_comp_character false "Katana_Knockback"))))

(event Ability|EventFailedToActivatePassiveAbility
  (params (ActorInfo :type Gameplay%20Ability%20Actor%20Info%20Structure))
  (|CallParentFunction ActorInfo))

(event EventDoLaunch
  (params (Vel :type Vector :default "0, 0, 0"))
  (|CallParentFunction Vel)
  (bind _returnvalue (GameplayTags|HasAnyMatchingGameplayTags(Message) (Variables|KnockbackParent|GetTarget) (Variables|Default|GetIgnoreFallDamageImmunityTags)))
  (if (not _returnvalue)
    (Ability|ApplyGameplayEffectToOwner (Variables|Default|GetGE_FallDamageImmunity)) ; Grant fall damage immunity 'til land
  ))

(event EventLaunchBuldingActors
  (params (Actor :type Actor%20Object%20Reference))
  (Ability|EndAbility))
